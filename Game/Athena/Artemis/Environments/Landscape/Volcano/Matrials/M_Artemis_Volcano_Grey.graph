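# Grey volcano terrain material for Artemis meshes, including spline meshes, with a WalkMesh switch selecting an alternate cliff-and-ground look.
# Layers tiled lava rock, cliff, crack and detail textures using overlay masks, vertex color, slope and height, adding tints and position-masked moss.
# Darkens roughness and specular below a waterline, then height-blends into the Apollo landscape runtime virtual texture.
# Mobile, low quality and the HLOD proxy skip the virtual texture blend.
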Material /Game/Athena/Artemis/Environments/Landscape/Volcano/Matrials/M_Artemis_Volcano_Grey
graph {
 properties {
  bTangentSpaceNormal=false
  bUseMaterialAttributes=true
  bUsedWithSplineMeshes=true
  pixelDepthOffsetMode=PDOM_Legacy
  usedShadingModels=Default Lit
 }
 nodes [
  n1 = RuntimeVirtualTextureSampleParameter() {materialType=BaseColor_Normal_Specular_YCoCg, parameterName=RVTD, virtualTexture=/Game/Athena/Apollo/Maps/Landscape/VT/RVT_Apollo_Landscape}
  n2 = SetMaterialAttributes!(MaterialAttributes, Base Color=@n1.BaseColor, Specular=@n1.BaseColor, Roughness=@n1.BaseColor, Normal=@n1.BaseColor) {attributeSetTypes [
    #1
    #2
    #3
    #4
   ]}
  n3 = StaticSwitchParameter(True=Constant() {r=15}, False=ScalarParameter() {defaultValue=150, group=RVT, parameterName=RVTBlendVert}) {parameterName=WalkMesh}
  n4 = ScalarParameter() {defaultValue=250, group=RVT, parameterName=RVTBlend}
  n5 = NamedRerouteUsage()
  n6 = NamedRerouteDeclaration(None=Divide(A=NamedRerouteUsage(), B=LinearInterpolate(A=@n3, B=@n4, Alpha=@n5))) {name=RVTDist, nodeColor=<rgba 1 0 0.8906 1>}
  n7 = NamedRerouteUsage()
  n8 = Constant() {r=3}
  n9 = ScalarParameter() {defaultValue=-2, group=RVT, parameterName=HMin}
  n10 = ScalarParameter() {defaultValue=3.2715, group=RVT, parameterName=HMax}
  n11 = NamedRerouteUsage()
  n12 = Saturate(None=OneMinus(None=NamedRerouteUsage()))
  n13 = StaticSwitchParameter(
   True=Saturate(None=Multiply(A=@n7, B=@n8))
   False=Max(A=LinearInterpolate(A=@n9, B=@n10, Alpha=@n11), B=LinearInterpolate(A=@n7, Alpha=@n12))
  ) {parameterName=WalkMesh}
  n14 = ScalarParameter() {defaultValue=0.6, group=RVT, parameterName=H Con}
  n15 = Add(A=VertexColor().R, B=ComponentMask(None=VertexNormalWS()) {b=true})
  n16 = LinearInterpolate(A=LinearInterpolate(A=Constant3Vector() {constant=<rgba 0.5677 0.4649 0.3893 0>}, Alpha=Saturate(None=@n15)), Alpha=NamedRerouteUsage())
  n17 = VectorParameter() {parameterName=GlobalTone, defaultValue=<rgba 0.09 0.0813 0.0695 1>}
  n18 = Multiply(A=@n17.RGB, B=Constant() {r=1.5})
  n19 = VertexColor()
  n20 = LinearInterpolate(
   A=ShadingPathSwitch(
    Default=StaticSwitchParameter(True=@n18, False=@n17.RGB) {parameterName=WalkMesh}
    Mobile=Constant3Vector() {constant=<rgba 0.2969 0.2873 0.2391 0>}
   )
   B=VectorParameter() {parameterName=TopTint, defaultValue=<rgba 0.394 0.1769 0.132 1>}.RGB
   Alpha=Multiply(A=NamedRerouteUsage(), B=OneMinus(None=@n19.B))
  )
  n21 = NamedRerouteUsage()
  n22 = VectorParameter() {parameterName=EdgeColor Outside, defaultValue=<rgba 0.1128 0.0919 0.0776 0>}
  n23 = Multiply(A=Constant3Vector() {desc=(R=0.637153,G=0.637153,B=0.637153,A=1.000000), constant=<rgba 0.5 0.4959 0.4791 1>}, B=@n20)
  n24 = NamedRerouteUsage()
  n25 = Constant() {r=0.5}
  n26 = Saturate(None=MaterialFunctionCall(In=Add(A=@n24, B=@n25), Contrast=NamedRerouteUsage()) {materialFunction=&CheapContrast}.Result)
  n27 = LinearInterpolate(
   A=LinearInterpolate(
    A=@n20
    B=MaterialFunctionCall(Base=@n20, Blend=@n21) {materialFunction=&Blend_Overlay}.Result
    Alpha=Constant() {r=0.8}
   )
   B=ShadingPathSwitch(Default=LinearInterpolate(A=@n22.RGB, B=@n23, Alpha=@n26), Mobile=@n23)
   Alpha=NamedRerouteUsage()
  )
  n28 = VectorParameter() {parameterName=Moss, defaultValue=<rgba 0.102 0.1059 0.0678 1>}
  n29 = Constant3Vector() {constant=<rgba -405.9166 -211.0814 0 0>}
  n30 = WorldPosition()
  n31 = NamedRerouteDeclaration(None=SphereMask(
    A=Add(A=@n29, B=@n30.XYZ)
    B=VectorParameter() {parameterName=Location, defaultValue=<rgba 42833.0469 56572.6094 4976 0>}.RGB
    Radius=Constant() {r=8141.7344}
    Hardness=Constant() {r=0.85}
   )) {name=SphereMask, nodeColor=<rgba 1 0 0.0469 1>}
  n32 = Saturate(None=ComponentMask(None=NamedRerouteUsage()) {b=true})
  n33 = Constant() {r=2}
  n34 = LinearInterpolate(
   A=Multiply(A=ScalarParameter() {defaultValue=0.65, parameterName=SideMoss}, B=@n31)
   B=Multiply(A=@n31, B=ScalarParameter() {defaultValue=0.25, parameterName=UPMoss})
   Alpha=Multiply(
    A=NamedRerouteUsage()
    B=MaterialFunctionCall(In=@n32, Contrast=@n33) {materialFunction=&CheapContrast}.Result
   )
  )
  n35 = WorldPosition()
  n36 = Divide(A=Add(A=ComponentMask(None=@n35.XYZ) {b=true}, B=Constant() {r=-5799.1147}), B=Constant() {r=-845.9937})
  n37 = Saturate(None=StaticSwitch(True=Multiply(A=@n34, B=@n36), False=VertexColor().G, Value=StaticBool()))
  n38 = ShadingPathSwitch(
   Default=MaterialFunctionCall(Base=LinearInterpolate(A=@n27, B=@n28.RGB, Alpha=@n37), Blend=NamedRerouteUsage()) {materialFunction=&Blend_Overlay}.Result
   Mobile=@n27
  )
  n39 = NamedRerouteUsage()
  n40 = MaterialFunctionCall(Base=@n38, Blend=@n39) {materialFunction=&Blend_Overlay}
  n41 = VectorParameter() {group=EdgeDirt, parameterName=GroundLineColor, defaultValue=<rgba 0.1024 0.08 0.0523 0>}
  n42 = NamedRerouteUsage()
  n43 = MaterialFunctionCall(Base=@n38, Blend=LinearInterpolate(B=NamedRerouteUsage(), Alpha=Constant() {r=0.5}) {constA=0.5, constB=0.5}) {materialFunction=&Blend_Overlay}
  n44 = StaticSwitchParameter(
   True=MaterialFunctionCall!(NoRVT=@n40.Result, RVT=LinearInterpolate(A=@n41.RGB, B=@n40.Result, Alpha=@n42)) {materialFunction=/Game/Packages/Fortress_SharedMaterials/Base_Material_Functions/FN_RVT_Switch_Color}.Result
   False=@n43.Result
  ) {parameterName=WalkMesh}
  n45 = NamedRerouteUsage()
  n46 = MaterialFunctionCall(Base=@n39, Blend=NamedRerouteUsage()) {materialFunction=&Blend_Overlay}
  n47 = Constant() {r=0.65}
  n48 = Constant() {r=2}
  n49 = NamedRerouteUsage()
  n50 = NamedRerouteUsage()
  n51 = LinearInterpolate(A=Multiply(A=LinearInterpolate(A=@n47, B=@n48, Alpha=@n49), B=OneMinus(None=@n50)), B=Constant() {r=1}, Alpha=NamedRerouteUsage())
  n52 = LinearInterpolate(B=OneMinus(None=NamedRerouteUsage()), Alpha=VertexColor().R)
  n53 = NamedRerouteUsage()
  n54 = Multiply(A=NamedRerouteUsage(), B=Constant3Vector() {constant=<rgba 1 1 0 1>})
  n55 = TextureSampleParameter2D() {
   parameterName=Normals
   samplerSource=SSM_Wrap_WorldGroupSettings
   samplerType=SAMPLERTYPE_Normal
   texture=/Game/Athena/Artemis/Environments/Meshes/Blockout/Volcano/S_Volcano_1_N
  }
  n56 = Multiply(A=NamedRerouteUsage(), B=ScalarParameter() {defaultValue=2, parameterName=Scaletile2})
  n57 = TextureSample(UVs=@n56) {
   samplerSource=SSM_Wrap_WorldGroupSettings
   samplerType=SAMPLERTYPE_Normal
   texture=/Game/Environments/Artemis/Sets/Terrain/Textures/T_Artemis_Lava_Rock_03_N
  }
  n58 = NamedRerouteDeclaration(None=@n57.RGB) {name=Lava Edge N, nodeColor=<rgba 0 1 0.25 1>}
  n59 = Divide(A=ObjectRadius()) {constB=1024}
  n60 = TextureCoordinate()
  n61 = NamedRerouteDeclaration(None=Multiply(A=Multiply(A=@n59, B=@n60), B=ScalarParameter() {defaultValue=1.52, parameterName=ScaleTile})) {name=ScaleTile, nodeColor=<rgba 0 0.1797 1 1>}
  n62 = TextureSample(UVs=@n61) {
   samplerSource=SSM_Wrap_WorldGroupSettings
   samplerType=SAMPLERTYPE_Normal
   texture=/Game/Environments/Artemis/Sets/Terrain/Textures/T_Artemis_Volcano_Cliff_02_N
  }
  n63 = TextureSample(UVs=Multiply(A=@n61, B=Constant() {r=5})) {
   samplerSource=SSM_Wrap_WorldGroupSettings
   samplerType=SAMPLERTYPE_Normal
   texture=&T_Artemis_Volcano_Cliff_02_N
  }
  n64 = Constant() {r=0.5}
  n65 = Add(A=@n62.RGB, B=Multiply(A=@n63.RGB, B=@n64))
  n66 = NamedRerouteDeclaration(None=ShadingPathSwitch(Default=@n65, Mobile=@n62.RGB)) {name=Lava_N, nodeColor=<rgba 0 0.0156 1 1>}
  n67 = WorldPosition()
  n68 = Divide(A=Add(A=ComponentMask(None=@n67.XYZ) {b=true}, B=Constant() {r=-3507.1582}), B=Constant() {r=4457.3916})
  n69 = Divide(A=NamedRerouteUsage(), B=ScalarParameter() {defaultValue=530.5585, desc=5622.774902, parameterName=RVT Edge Details Lenght})
  n70 = NamedRerouteDeclaration(None=Saturate(None=MaterialFunctionCall(NoRVT=@n68, RVT=@n69) {materialFunction=&FN_RVT_Switch_Color}.Result)) {name=Edge, nodeColor=<rgba 0.5469 0 1 1>}
  n71 = ShadingPathSwitch(Default=Multiply(A=LinearInterpolate(A=@n58, B=@n66, Alpha=@n70), B=Constant3Vector() {constant=<rgba 2 2 0 0>}), Mobile=@n66)
  n72 = NamedRerouteUsage()
  n73 = Multiply(A=TextureCoordinate(), B=ScalarParameter() {defaultValue=4, parameterName=ScaleCracks})
  n74 = TextureSample(UVs=@n73) {
   samplerSource=SSM_Wrap_WorldGroupSettings
   samplerType=SAMPLERTYPE_Normal
   texture=/Game/Environments/Artemis/Sets/Terrain/Textures/T_Artemis_Lava_Rock_02_N
  }
  n75 = TextureSample(UVs=@n73) {samplerSource=SSM_Wrap_WorldGroupSettings, texture=/Game/Environments/Artemis/Sets/Terrain/Textures/T_Artemis_Lava_Rock_02_D}
  n76 = MaterialFunctionCall(
   A=Multiply(A=@n74.RGB, B=Constant() {r=2})
   B=Multiply(A=Constant() {r=0.35}, B=NamedRerouteUsage())
   Transition Phase=Constant() {r=0.5}
   Height Texture=Max(A=NamedRerouteDeclaration(None=@n75.A) {name=CracksA, nodeColor=<rgba 0 1 0.8359 1>}, B=NamedRerouteUsage())
   Contrast=Constant() {r=2}
  ) {materialFunction=&HeightLerp}
  n77 = Constant() {r=0.5}
  n78 = Multiply(A=NamedRerouteUsage(), B=VertexColor().R)
  n79 = Constant() {r=3}
  n80 = Add(
   A=@n55.RGB
   B=Multiply(
    A=Constant3Vector() {constant=<rgba 1.25 1.25 0 1>}
    B=MaterialFunctionCall(A=@n72, B=@n76.Results, Transition Phase=@n77, Height Texture=@n78, Contrast=@n79) {materialFunction=&HeightLerp}.Results
   )
  )
  n81 = NamedRerouteUsage()
  n82 = LinearInterpolate(
   A=Add(A=@n55.RGB, B=NamedRerouteDeclaration(None=@n71) {name=Lava and Edge, nodeColor=<rgba 0 0.4141 1 1>})
   B=@n80
   Alpha=Saturate(None=OneMinus(None=@n81))
  )
  n83 = VertexColor()
  n84 = ShadingPathSwitch(Default=Add(A=@n82, B=LinearInterpolate(A=@n54, B=@n82, Alpha=@n83.A)), Mobile=@n82)
  n85 = SetMaterialAttributes(
   Base Color=ShadingPathSwitch(Default=Multiply(A=@n16, B=@n44), Mobile=Multiply(A=@n44, B=@n45))
   Specular=ShadingPathSwitch(Default=StaticSwitchParameter(True=@n46.Result, False=@n43.Result) {parameterName=WalkMesh}, Mobile=Constant() {r=0.25})
   Roughness=StaticSwitchParameter(True=NamedRerouteUsage(), False=Add(A=@n51, B=@n52)) {parameterName=WalkMesh}
   Normal=Transform(None=StaticSwitchParameter(True=@n53, False=@n84) {parameterName=WalkMesh})
  ) {attributeSetTypes [
    #1
    #2
    #3
    #4
   ]}
  n86 = GetMaterialAttributes(None=@n85) {attributeGetTypes [#2, #3, #1]}
  n87 = SetMaterialAttributes!(
   MaterialAttributes=@n86.MaterialAttributes
   Specular=Multiply(A=@n86.Specular) {constB=0.5}
   Roughness=Multiply(A=@n86.Roughness, B=Constant() {r=0.1}) {constB=0.1}
   Base Color=@n86.MaterialAttributes
  ) {attributeSetTypes [#2, #3, #1]}
  n88 = WorldPosition()
  n89 = Divide(A=Add(A=ComponentMask(None=@n88.XYZ) {b=true}, B=Constant() {r=-3075.6204}), B=Constant() {r=355})
  n90 = ComponentMask(None=VertexNormalWS()) {b=true}
  n91 = Saturate(None=Subtract(A=NamedRerouteUsage(), B=Add(A=@n89, B=@n90)))
  n92 = ShadingPathSwitch(Default=QualitySwitch(Default=BlendMaterialAttributes(A=@n87, B=@n85, Alpha=@n91), Low=@n85), Mobile=@n85)
  n93 = StaticSwitchParameter(
   True=SetMaterialAttributes!(MaterialAttributes=@n92, Base Color=NamedRerouteUsage()) {attributeSetTypes [#1]}
   False=@n92
  ) {parameterName=DebugScale}
  n94 = MaterialFunctionCall(
   NoRVT=@n93
   RVT=BlendMaterialAttributes(
    A=@n2
    B=@n93
    Alpha=MaterialFunctionCall(Transition Phase=@n6, Height Texture=@n13, Contrast=@n14) {materialFunction=&HeightLerp}.Alpha
   )
  ) {materialFunction=&FN_RVT_Switch_Material}
  MaterialProxyReplace(
   Realtime=ShadingPathSwitch(Default=QualitySwitch(Default=@n94.Result, Low=@n93, High=@n94.Result, Medium=@n94.Result, Epic=@n94.Result), Mobile=@n93)
   MaterialProxy=SetMaterialAttributes!(MaterialAttributes=@n93, Normal=NamedRerouteUsage()) {attributeSetTypes [#4]}
  ) {desc=Don't use RVT on HLODs}
  n95 = WorldPosition()
  n96 = Constant() {r=3500}
  n97 = RuntimeVirtualTextureSampleParameter() {materialType=WorldHeight, parameterName=RVTH, virtualTexture=/Game/Athena/Apollo/Maps/Landscape/VT/RVT_Apollo_Landscape_Height}
  NamedRerouteDeclaration(None=Subtract(
    A=ComponentMask(None=@n95.XYZ) {b=true}
    B=MaterialFunctionCall(NoRVT=@n96, RVT=@n97.WorldHeight) {materialFunction=&FN_RVT_Switch_Color}.Result
   )) {name=RVTH, nodeColor=<rgba 0 0.7187 1 1>}
  NamedRerouteDeclaration(None=Transform(None=@n55.RGB)) {name=WS N, nodeColor=<rgba 1 0.3281 0 1>}
  n98 = ComponentMask(None=WorldPosition().XYZ) {b=true}
  n99 = ScalarParameter() {defaultValue=-2874.5847, parameterName=WaterlineAdd}
  n100 = Saturate(None=Divide(A=Add(A=@n98, B=@n99), B=ScalarParameter() {defaultValue=388.4267, parameterName=WaterlineDiv}))
  NamedRerouteDeclaration(None=LinearInterpolate(
    A=ScalarParameter() {defaultValue=0.258, parameterName=WaterLineHeight0}
    B=ScalarParameter() {defaultValue=0.588, parameterName=WaterLineHeight1}
    Alpha=NamedRerouteDeclaration(None=@n100) {name=Waterheight, nodeColor=<rgba 0.8984 0 1 1>}
   )) {name=Waterline, nodeColor=<rgba 0 0.0156 1 1>}
  n101 = MaterialFunctionCall!(
   Diffuse Texture Object=TextureObject() {samplerType=SAMPLERTYPE_LinearGrayscale, texture=/Game/Environments/Artemis/Sets/Terrain/Textures/T_Artemis_Volcano_Cliff_01_H}
   Spec Texture Object=TextureObject() {samplerType=SAMPLERTYPE_Masks, texture=/Game/Environments/Artemis/Sets/Terrain/Textures/T_Artemis_Volcano_Cliff_01_S}
   TextureSize=ScalarParameter() {defaultValue=450, parameterName=Scale Cliff}
   --------------------------
   Use seperate settings for Z projection
   Use seperate texture for Z projection
   Texture Size Z projection
   Z Plane Diffuse Texture Object
   Z Plane Spec Texture Object
   "--------------------------      "
   ProjectionTransitionContrast
   Use Texture Alpha Heightmap to Refine the Transistion
   HeightLerpTransitionContrast
   PickTheDiffuseHeightChannel
   TheDiffuseHeightChannel
   "--------------------------       "
   NormalTextureObject=TextureObject() {samplerType=SAMPLERTYPE_Normal, texture=/Game/Environments/Artemis/Sets/Terrain/Textures/T_Artemis_Volcano_Cliff_01_N}
   World Space Normals=StaticBool()
   "--------------------------  "
   Export Float 4 Diffuse=StaticBool()
   Export Float 4 Spec
   "--------------------------     "
   WorldPosition
   World Space Normal
  ) {materialFunction=/Engine/Functions/Engine_MaterialFunctions02/Texturing/WorldAlignedTextures_Complex}
  NamedRerouteDeclaration(None=@n101.XY Spec) {name=Cliff Spec, nodeColor=<rgba 0 1 0.5078 1>}
  NamedRerouteDeclaration(None=@n101.XY Diffuse Texture) {name=Cliff D, nodeColor=<rgba 1 0.1406 0 1>}
  NamedRerouteDeclaration(None=@n101.Normal XY Texture) {name=Cliff N, nodeColor=<rgba 1 0.5625 0 1>}
  n102 = Constant() {r=1.7459}
  n103 = Constant() {r=0.8712}
  n104 = Saturate(None=ComponentMask(None=NamedRerouteUsage()) {b=true})
  NamedRerouteDeclaration(None=Saturate(None=MaterialFunctionCall!(ThresholdValue=@n102, ThresholdRange=@n103, Alpha=@n104, TwoSidedThreshold) {materialFunction=/Engine/Functions/Engine_MaterialFunctions02/Masking/ThresholdWithRange}.Result)) {name=AngleMask, nodeColor=<rgba 1 0.3516 0 1>}
  n105 = Multiply(A=NamedRerouteUsage(), B=ScalarParameter() {defaultValue=4, parameterName=TileDetail})
  n106 = TextureSample(UVs=Multiply(A=@n105) {constB=2}) {
   samplerSource=SSM_Wrap_WorldGroupSettings
   samplerType=SAMPLERTYPE_Masks
   texture=/Game/Environments/Apollo/Sets/CubeTownV2/Materials/Parent/T_CubeTownDirt_M
  }
  n107 = TextureSample(UVs=@n105) {
   samplerSource=SSM_Wrap_WorldGroupSettings
   samplerType=SAMPLERTYPE_Masks
   texture=&T_CubeTownDirt_M
  }
  n108 = Constant() {r=0.25}
  NamedRerouteDeclaration(None=StaticSwitchParameter(
    True=Min(A=@n106.R, B=@n107.R)
    False=MaterialFunctionCall(In=@n106.R, Contrast=@n108) {materialFunction=&CheapContrast}.Result
   ) {parameterName=WalkMesh}) {name=DetailsD, nodeColor=<rgba 1 0 0.4219 1>}
  n109 = TextureSample(UVs=@n105) {
   samplerSource=SSM_Wrap_WorldGroupSettings
   samplerType=SAMPLERTYPE_Normal
   texture=/Game/Environments/Apollo/Rocks/Natural_Rocks/Textures/T_RockDetails_02_N
  }
  n110 = Constant3Vector() {constant=<rgba 1 1 0 0>}
  NamedRerouteDeclaration(None=ShadingPathSwitch(Default=Multiply(A=@n109.RGB, B=@n110), Mobile=Constant3Vector() {constant=<rgba 0 0 1 0>})) {name=Details N, nodeColor=<rgba 1 0 0.1641 1>}
  NamedRerouteDeclaration(None=TextureSample(UVs=@n61) {samplerSource=SSM_Wrap_WorldGroupSettings, texture=/Game/Environments/Artemis/Sets/Terrain/Textures/T_Artemis_Volcano_Cliff_02_D}.R) {name=Lava_D, nodeColor=<rgba 0 0.0156 1 1>}
  n111 = TextureSampleParameter2D() {
   parameterName=Normals
   samplerSource=SSM_Wrap_WorldGroupSettings
   samplerType=SAMPLERTYPE_Normal
   texture=&S_Volcano_1_N
  }
  n112 = NamedRerouteUsage()
  n113 = Constant3Vector() {constant=<rgba 2 2 0 0>}
  n114 = Constant3Vector() {constant=<rgba 1 1 0 0>}
  n115 = Add(A=@n111.RGB, B=ShadingPathSwitch(Default=Multiply(A=@n112, B=@n113), Mobile=Multiply(A=@n112, B=@n114)))
  n116 = NamedRerouteUsage()
  n117 = Constant3Vector() {constant=<rgba 0.25 0.25 0 0>}
  n118 = MaterialFunctionCall(
   A=@n115
   B=Add(A=@n111.RGB, B=Multiply(A=@n116, B=@n117))
   Transition Phase=NamedRerouteUsage()
   Height Texture=ComponentMask(None=NamedRerouteUsage()) {r=true}
   Contrast=Constant()
  ) {materialFunction=&HeightLerp}
  NamedRerouteDeclaration(None=@n118.Alpha) {name=CliffMask, nodeColor=<rgba 1 0.9609 0 1>}
  n119 = VectorParameter() {parameterName=CliffCracks, defaultValue=<rgba 0.1068 0.3439 0.3594 1>}
  n120 = ComponentMask(None=Add(A=WorldPosition().XYZ, B=Constant() {r=-3040})) {b=true}
  n121 = Constant() {r=512}
  n122 = LinearInterpolate(
   A=@n119.RGB
   B=VectorParameter() {parameterName=CliffCracks2, defaultValue=<rgba 0.5295 0.2358 0.148 1>}.RGB
   Alpha=Saturate(None=Divide(A=@n120, B=@n121))
  )
  n123 = ShadingPathSwitch(Default=@n122, Mobile=@n119.RGB)
  n124 = NamedRerouteUsage()
  n125 = LinearInterpolate(A=VertexNormalWS(), B=NamedRerouteUsage(), Alpha=Constant() {r=0.6})
  n126 = MaterialFunctionCall(
   Transition Phase=Saturate(None=ComponentMask(None=@n125) {b=true})
   Height Texture=ComponentMask(None=NamedRerouteUsage()) {g=true}
   Contrast=Constant() {r=0.5}
  ) {materialFunction=&HeightLerp}
  n127 = LinearInterpolate(
   A=LinearInterpolate(A=NamedRerouteDeclaration(None=@n123) {name=WaterColor, nodeColor=<rgba 1 0 0.3984 1>}, B=@n124, Alpha=NamedRerouteUsage())
   B=Min(A=@n124, B=NamedRerouteUsage())
   Alpha=@n126.Alpha
  ) {constB=0.5}
  n128 = LinearInterpolate(A=Constant() {r=1}, B=Constant() {r=2}, Alpha=ComponentMask(None=VertexNormalWS()) {b=true})
  NamedRerouteDeclaration(None=ShadingPathSwitch(Default=@n127, Mobile=Multiply(A=@n127, B=@n128))) {name=WalkMeshDiffuse, nodeColor=<rgba 0 1 0.1562 1>}
  NamedRerouteDeclaration(None=TextureSample(UVs=@n61) {texture=/Game/FN_PublicTurntable/Turntable_Lowpoly/Content/BlueprintAssets/Checker}.RGB) {name=Debug Scale, nodeColor=<rgba 0 0.0156 1 1>}
  n129 = NamedRerouteUsage()
  n130 = ScalarParameter() {defaultValue=133.7615, group=EdgeDirt, parameterName=EdgeDirtHeight}
  NamedRerouteDeclaration(None=Saturate(None=Divide(A=@n129, B=@n130))) {name=EdgeDiffuseMask, nodeColor=<rgba 1 0.5625 0 1>}
  n131 = MaterialFunctionCall(ThresholdValue=Constant() {r=1.1}, ThresholdRange=Constant() {r=0.5}, Alpha=ComponentMask(None=VertexNormalWS()) {b=true}) {materialFunction=&ThresholdWithRange}
  NamedRerouteDeclaration(None=Round(None=Saturate(None=@n131.Result))) {name=SidesMask, nodeColor=<rgba 1 0.2109 0 1>}
  n132 = TextureSampleParameter2D() {
   parameterName=Overlay
   samplerSource=SSM_Wrap_WorldGroupSettings
   samplerType=SAMPLERTYPE_Masks
   texture=/Game/Athena/Artemis/Environments/Landscape/Volcano/Textures/Volcano_A_M
  }
  NamedRerouteDeclaration(None=@n132.B) {name=Mask B, nodeColor=<rgba 0 0.0066 1 1>}
  NamedRerouteDeclaration(None=Multiply(A=@n65, B=@n64)) {name=Laval_N_Mix, nodeColor=<rgba 1 0 0.6328 1>}
  NamedRerouteDeclaration(None=@n80) {name=HLOD_Normal, nodeColor=<rgba 0 0.2031 1 1>}
  NamedRerouteDeclaration(None=@n132.R) {name=Mask R, nodeColor=<rgba 1 0 0.0636 1>}
  n133 = WorldPosition()
  n134 = Divide(A=Add(A=ComponentMask(None=@n133.XYZ) {b=true}, B=Constant() {r=-3919.9656}), B=Constant() {r=6451.4448})
  n135 = Constant() {desc=5622.774902, r=3000}
  n136 = Divide(A=NamedRerouteUsage(), B=LinearInterpolate(A=@n135, B=Multiply(A=@n135) {constA=2, constB=2}, Alpha=VertexColor().R))
  n137 = Multiply(A=NamedRerouteDeclaration(None=@n132.G) {name=Mask G, nodeColor=<rgba 0.1016 1 0 1>}, B=Constant() {r=2})
  NamedRerouteDeclaration(None=MaterialFunctionCall(
    Transition Phase=MaterialFunctionCall(NoRVT=@n134, RVT=@n136) {materialFunction=&FN_RVT_Switch_Color}.Result
    Height Texture=NamedRerouteDeclaration(None=@n137) {name=Boosted Mask, nodeColor=<rgba 0.4766 1 0 1>}
    Contrast=Constant() {desc=0.35, r=-3}
   ) {materialFunction=&HeightLerp}.Alpha) {name=Slope, nodeColor=<rgba 0 1 0.9531 1>}
  NamedRerouteDeclaration(None=TextureSample(UVs=@n56) {samplerSource=SSM_Wrap_WorldGroupSettings, texture=/Game/Environments/Artemis/Sets/Terrain/Textures/T_Artemis_Lava_Rock_03_D}.A) {name=ScreeA, nodeColor=<rgba 0.6875 0 1 1>}
  n138 = NamedRerouteUsage()
  n139 = TextureSampleParameter2D() {
   parameterName=Normals
   samplerSource=SSM_Wrap_WorldGroupSettings
   samplerType=SAMPLERTYPE_Normal
   texture=&S_Volcano_1_N
  }
  n140 = NamedRerouteUsage()
  n141 = Constant3Vector() {constant=<rgba 1 1 0 0>}
  n142 = Constant() {r=2}
  n143 = NamedRerouteUsage()
  n144 = LinearInterpolate(
   A=Add(A=Add(A=@n138, B=@n139.RGB), B=Multiply(A=@n140, B=@n141))
   B=Add(A=Multiply(A=@n138, B=@n142), B=Add(A=@n143, B=@n139.RGB))
   Alpha=NamedRerouteUsage()
  )
  NamedRerouteDeclaration(None=LinearInterpolate(A=ShadingPathSwitch(Default=@n118.Results, Mobile=@n115), B=ShadingPathSwitch(Default=@n144, Mobile=@n139.RGB), Alpha=@n126.Alpha)) {name=WalkMeshNormal, nodeColor=<rgba 0 1 0.1562 1>}
  n145 = NamedRerouteUsage()
  n146 = LinearInterpolate(A=Constant() {r=-0.6227}, B=Constant() {r=1.6501}, Alpha=NamedRerouteUsage())
  n147 = Constant() {r=0.6}
  NamedRerouteDeclaration(None=LinearInterpolate(A=ComponentMask(None=@n145) {b=true}, B=ShadingPathSwitch(Default=@n146, Mobile=@n147), Alpha=@n126.Alpha)) {name=WalkMeshRough, nodeColor=<rgba 0 1 0.1562 1>}
  Reroute(None=@n122)
 ]
}
